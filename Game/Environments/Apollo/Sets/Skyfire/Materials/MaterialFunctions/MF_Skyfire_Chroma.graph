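# Skyfire material function that produces a chroma color from the reflection vector, defaulting to an anodized reflection color curve.
# Offset red and green components of the reflection vector collapse through a dot product into the lookup time for a parameterized curve atlas row.
# The curve color is multiplied by a color mask, an exponent-shaped ambient occlusion mask, and an inverted Fresnel from the pixel normal.
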MaterialFunction /Game/Environments/Apollo/Sets/Skyfire/Materials/MaterialFunctions/MF_Skyfire_Chroma
graph {nodes [
  n1 = FunctionInput() {bUsePreviewValueAsDefault=true, inputName=AmbientOcclusionExponent, inputType=FunctionInput_Scalar, sortPriority=1, previewValue=<4 0 0 1>}
  n2 = Power(Base=FunctionInput() {bUsePreviewValueAsDefault=true, inputName=AmbientOcclusionMask, previewValue=<1 1 1 1>}, Exp=@n1) {constExponent=4}
  n3 = Subtract(A=FunctionInput(Preview=ReflectionVectorWS()) {bUsePreviewValueAsDefault=true, inputName=ReflectionVector, sortPriority=3}) {constB=-0.5}
  n4 = CurveAtlasRowParameter(CurveTime=DotProduct(A=ComponentMask(None=@n3) {g=true, r=true}, B=Constant() {r=0.5})) {
   atlas=/Game/Environments/Apollo/Sets/Skyfire/Curves/SkyfireCurveAtlas
   curve=/Game/Environments/Apollo/Sets/Skyfire/Curves/C_Skyfire_AnodizedReflectionColor
   defaultValue=1
   group=Chroma
   parameterName=ChromaColorCurve
  }
  n5 = Multiply(A=FunctionInput() {bUsePreviewValueAsDefault=true, inputName=ColorMask, sortPriority=2, previewValue=<1 1 1 1>}, B=@n4)
  n6 = FunctionInput() {bUsePreviewValueAsDefault=true, inputName=FresnelMaskExponent, inputType=FunctionInput_Scalar, sortPriority=4, previewValue=<4 0 0 1>}
  n7 = PixelNormalWS()
  FunctionOutput(None=Multiply(
    A=Multiply(A=@n2, B=@n5)
    B=MaterialFunctionCall(Exponent=@n6, Vertex / Pixel Normal=@n7) {materialFunction=&MF_Fresnel_FX}.Fresnel Inverted
   ))
 ]}
