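# Skyfire set material function that spins geometry about an axis, outputting a RotateAboutAxis offset from world position plus the axis, angle and pivot it used.
# Axis and pivot parameters are converted from local to world space; axis and angle are overridable inputs.
# The default angle comes from MF_ModuloTime with a parameterized divisor, animating rotation over time.
# A Normalize node on the axis is unconnected, so the axis passes through unnormalized.
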MaterialFunction /Game/Environments/Apollo/Sets/Skyfire/Materials/MaterialFunctions/MF_Skyfire_RotateAboutAxis
graph {nodes [
  n1 = Transform(None=FunctionInput(Preview=VectorParameter() {group=RotateAboutAxis, parameterName=NormalizeRotationAxis, defaultValue=<rgba 0 0 1 0>}.RGB) {bUsePreviewValueAsDefault=true, inputName=NormalizeRotationAxis}) {transformSourceType=TRANSFORMSOURCE_Local}
  n2 = FunctionInput(Preview=MaterialFunctionCall!(Time, Divisor=ScalarParameter() {defaultValue=60, group=RotateAboutAxis, parameterName=RotateAboutAxisTimeDivisor}, Mod) {materialFunction=/Game/Effects/Fort_Effects/Materials/_Functions/MF_ModuloTime}.Unclamped) {bUsePreviewValueAsDefault=true, inputName=RotationAngleOverride, inputType=FunctionInput_Scalar, sortPriority=1}
  n3 = TransformPosition(None=VectorParameter() {group=RotateAboutAxis, parameterName=RotateAboutAxisPivotPointLocalSpace}.RGB) {transformType=TRANSFORMPOSSOURCE_World}
  FunctionOutput(None=RotateAboutAxis(
    NormalizedRotationAxis=FunctionOutput(None=@n1) {outputName=NormalizeRotationAxis, sortPriority=1}
    RotationAngle=FunctionOutput(None=@n2) {outputName=RotationAngle, sortPriority=2}
    PivotPoint=FunctionOutput(None=@n3) {outputName=PivotPoint, sortPriority=3}
    Position=WorldPosition().XYZ
   ))
  Normalize(VectorInput=@n1)
 ]}
