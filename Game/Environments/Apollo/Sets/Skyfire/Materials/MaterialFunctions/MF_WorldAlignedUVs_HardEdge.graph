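# Material function that derives hard-edged world-aligned planar UVs from position, choosing a projection plane by absolute normal components instead of blending.
# Position and normal default to world position and vertex normal, with a static switch converting both to local space.
# Nested Ifs select XY when Z exceeds X, otherwise choosing between YZ and XZ.
# As wired, the Max and inner If compare X with itself, so Y is never tested.
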MaterialFunction /Game/Environments/Apollo/Sets/Skyfire/Materials/MaterialFunctions/MF_WorldAlignedUVs_HardEdge
graph {nodes [
  n1 = FunctionInput(Preview=VertexNormalWS()) {bUsePreviewValueAsDefault=true, inputName=NormalOverride, sortPriority=2}
  n2 = Normalize(VectorInput=Transform(None=@n1) {transformSourceType=TRANSFORMSOURCE_World, transformType=TRANSFORM_Local})
  n3 = FunctionInput(Preview=StaticBool()) {bUsePreviewValueAsDefault=true, inputName=UseLocalSpace, inputType=FunctionInput_StaticBool}
  n4 = MaterialFunctionCall(Float3=Abs(None=StaticSwitch(True=@n2, False=@n1, Value=@n3))) {materialFunction=&BreakOutFloat3Components}
  n5 = FunctionInput(Preview=WorldPosition().XYZ) {bUsePreviewValueAsDefault=true, inputName=WorldPosition, sortPriority=1}
  n6 = StaticSwitch(True=TransformPosition(None=@n5) {transformSourceType=TRANSFORMPOSSOURCE_World}, False=@n5, Value=@n3)
  n7 = ComponentMask(None=@n6) {b=true, desc=YZ, g=true}
  n8 = ComponentMask(None=@n6) {b=true, desc=XZ, r=true}
  FunctionOutput(None=If(
    A=@n4.B
    B=Max(A=@n4.R, B=@n4.R)
    "A > B"=ComponentMask(None=@n6) {desc=YX, g=true, r=true}
    "A < B"=If(A=@n4.R, B=@n4.R, "A > B"=@n7, "A < B"=@n8)
   ))
 ]}
