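# Material function that builds the layered metal surface for the Skyfire environment set, output as material attributes.
# A mask texture and packed tiling textures on world-aligned UVs drive curve-atlas base color with view-dependent tech patterns, optionally animated roughness, and layered detail normals.
# Emissive combines curvature tint, texture or curve color, and fake specular from a fixed light vector.
# A "Temp fix" rotation switch adds WPO, adjusts UVs and normals, and disables Chroma.
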MaterialFunction /Game/Environments/Apollo/Sets/Skyfire/Materials/MaterialFunctions/MF_Skyfire_Metal
graph {nodes [
  n1 = SetMaterialAttributes!(
   MaterialAttributes
   BaseColor=NamedRerouteUsage()
   Metallic=NamedRerouteUsage()
   Specular=ScalarParameter() {defaultValue=1, group=Specular, parameterName=Specular}
   Roughness=NamedRerouteUsage()
   Anisotropy=ScalarParameter() {defaultValue=0.15, group=Anisotropy, parameterName=Anisotropy}
   EmissiveColor=NamedRerouteUsage()
   Normal=NamedRerouteUsage()
   Tangent=NamedRerouteUsage()
   CustomData0=ScalarParameter() {group=Clear Coat, parameterName=Clear Coat}
   CustomData1=ScalarParameter() {group=Clear Coat, parameterName=Clear Coat Roughness}
   AmbientOcclusion=NamedRerouteUsage()
   WorldPositionOffset=NamedRerouteUsage()
  ) {attributeSetTypes [
    #1
    #2
    #3
    #4
    #5
    #6
    #7
    #8
    #9
    #10
    #11
    #12
   ]}
  FunctionOutput(None=@n1)
  n2 = TextureSampleParameter2D() {parameterName=Masks, samplerType=SAMPLERTYPE_LinearColor, texture=/Game/Environments/Apollo/Sets/Skyfire/Textures/T_rearwall_MSK}
  NamedRerouteDeclaration(None=@n2.RGB) {name=MasksRGB, nodeColor=<rgba 0 0.0391 1 1>}
  NamedRerouteDeclaration(None=@n2.R) {name=MasksR_Glow, nodeColor=<rgba 0 0.6719 1 1>}
  n3 = WorldPosition()
  n4 = NamedRerouteUsage()
  n5 = Add(
   A=@n3.XYZ
   B=MaterialFunctionCall!(NormalizeRotationAxis, RotationAngleOverride=OneMinus(None=@n4)) {materialFunction=&/MF_Skyfire_RotateAboutAxis}.Result
  )
  n6 = NamedRerouteDeclaration(None=MaterialFunctionCall!(
    UseLocalSpace=StaticBoolParameter() {group=Tile, parameterName=UseLocalSpaceForTriPlanarTiling}
    WorldPosition=StaticSwitchParameter(True=@n5, False=@n3.XYZ) {desc=Temp fix, group=RotateAboutAxis, parameterName=UseRotation}
    NormalOverride
   ) {materialFunction=&/MF_WorldAlignedUVs_HardEdge}.Result) {name=CheapWorldAlignedUVs, nodeColor=<rgba 0 0.7656 1 1>}
  n7 = ScalarParameter() {defaultValue=220, group=Tile, parameterName=TechTileDivisor}
  NamedRerouteDeclaration(None=Divide(A=@n6, B=@n7)) {name=TechPatternUVs, nodeColor=<rgba 1 0.0234 0 1>}
  NamedRerouteDeclaration(None=@n2.RGBA) {name=MasksRGBA, nodeColor=<rgba 0 1 0.9531 1>}
  n8 = NamedRerouteUsage()
  n9 = NamedRerouteUsage()
  n10 = Constant()
  n11 = TextureCoordinate() {coordinateIndex=1}
  n12 = TextureCoordinate()
  n13 = TextureSampleParameter2D(UVs=StaticSwitchParameter(True=NamedRerouteUsage(), False=StaticSwitchParameter(True=@n11, False=@n12) {group=Normal, parameterName=NormalMapUseUV1}) {group=Normal, parameterName=NormalUsesMetalPatternUVs}) {group=Normal, parameterName=Normal, samplerType=SAMPLERTYPE_Normal, texture=/Game/Environments/Apollo/Sets/Skyfire/Textures/T_rear_wall_N}
  n14 = VectorParameter() {group=Normal, parameterName=NormalStrength, defaultValue=<rgba 1 1 1 0>}
  n15 = TextureObjectParameter() {
   parameterName=AdditionalDetailNormal
   samplerType=SAMPLERTYPE_Normal
   texture=/Game/Global/Textures/Default/Tileable_Patterns/T_BrushedMetal_Detail_N
  }
  n16 = NamedRerouteUsage()
  n17 = TextureObjectParameter() {parameterName=DetailNormal, samplerType=SAMPLERTYPE_Normal, texture=&T_BrushedMetal_Detail_N}
  n18 = StaticSwitchParameter(True=NamedRerouteUsage(), False=NamedRerouteUsage()) {group=Normal, parameterName=DetailNormalUsesTechPatternUVs}
  n19 = Multiply(
   A=NamedRerouteDeclaration(None=MaterialFunctionCall(TextureObject=@n17, TextureSize=@n18) {materialFunction=&WorldAlignedNormal}.XYZ Texture) {name=DetailNormal, nodeColor=<rgba 0 1 0.9531 1>}
   B=VectorParameter() {group=Normal, parameterName=DetailNormalStrength, defaultValue=<rgba 0.3 0.3 1 0>}.RGB
  )
  n20 = Add(
   A=Multiply(
    A=MaterialFunctionCall(TextureObject=@n15, TextureSize=@n16) {materialFunction=&WorldAlignedNormal}.XYZ Texture
    B=VectorParameter() {group=Normal, parameterName=AdditionalDetailNormalStrength, defaultValue=<rgba 1 1 1 0>}.RGB
   )
   B=@n19
  )
  n21 = Add(
   A=StaticSwitchParameter(True=Multiply(A=@n13.RGB, B=@n14.RGB), False=Constant3Vector() {constant=<rgba 0 0 1 0>}) {defaultValue=true, group=Normal, parameterName=UseNormalMap}
   B=Multiply(
    A=StaticSwitchParameter(True=@n20, False=@n19) {group=Normal, parameterName=UseAdditionalDetailNormal}
    B=Constant3Vector() {constant=<rgba 1 1 0 0>}
   )
  )
  n22 = Transform(None=@n21)
  n23 = Transform(None=Add(A=RotateAboutAxis(NormalizedRotationAxis=@n8, RotationAngle=@n9, PivotPoint=@n10, Position=@n22), B=@n22)) {transformSourceType=TRANSFORMSOURCE_World, transformType=TRANSFORM_Tangent}
  NamedRerouteDeclaration(None=Normalize(VectorInput=StaticSwitchParameter(True=@n23, False=@n21) {desc=Temp fix, group=RotateAboutAxis, parameterName=UseRotation})) {name=FinalNormalMap, nodeColor=<rgba 0.7109 1 0 1>}
  n24 = NamedRerouteUsage()
  n25 = Power(Base=NamedRerouteUsage(), Exp=ScalarParameter() {defaultValue=2, group=Emissive, parameterName=CurvatureEmissiveExponent})
  n26 = VectorParameter() {group=Emissive, parameterName=CurvatureColorEmissive, defaultValue=<rgba 0.3829 0.173 0.8976 0>}
  n27 = Multiply(
   A=TextureSampleParameter2D() {group=Emissive, parameterName=EmissiveTexture, texture=&T_BlacksRGB}.RGB
   B=VectorParameter() {group=Emissive, parameterName=EmissiveTextureScale, defaultValue=<rgba 1 1 1 0>}.RGB
  )
  n28 = StaticSwitchParameter(True=Multiply(A=NamedRerouteUsage(), B=@n27), False=@n27) {group=Emissive, parameterName=ModulateEmissiveTextureByGlowMask}
  n29 = CurveAtlasRowParameter() {
   atlas=/Game/Environments/Apollo/Sets/Skyfire/Curves/SkyfireCurveAtlas
   bUseCustomPrimitiveData=true
   curve=/Game/Environments/Apollo/Sets/Skyfire/Curves/C_Skyfire_Black
   defaultValue=9
   group=Emissive
   parameterName=EmissiveColorCurve
  }
  n30 = CurveAtlasRowParameter() {
   atlas=&SkyfireCurveAtlas
   curve=&C_Skyfire_Black
   defaultValue=9
   group=Emissive
   parameterName=EmissiveColorCurve
  }
  n31 = Multiply(A=NamedRerouteUsage(), B=StaticSwitchParameter(True=@n29, False=@n30) {group=Emissive, parameterName=EmissiveUsesCPD})
  n32 = StaticSwitchParameter(
   True=StaticSwitchParameter(True=Add(A=@n28, B=@n31), False=@n28) {group=Emissive, parameterName=AddEmissiveColorCurveToEmissiveTexture}
   False=@n31
  ) {group=Emissive, parameterName=UseEmissiveTexture}
  n33 = VectorParameter() {group=Emissive, parameterName=Mobile_EmissiveMult, defaultValue=<rgba 1 1 1 1>}
  n34 = Add(A=Add(A=Multiply(A=@n25, B=@n26.RGB), B=NamedRerouteUsage()), B=ShadingPathSwitch(Default=@n32, Mobile=Multiply(A=@n32, B=@n33.RGB)))
  NamedRerouteDeclaration(None=StaticSwitchParameter(True=Add(A=@n24, B=@n34), False=@n34) {defaultValue=true, group=FakeSpecular, parameterName=UseFakeSpecular}) {name=Emissive, nodeColor=<rgba 0 0.25 1 1>}
  n35 = TextureSampleParameter2D(UVs=NamedRerouteUsage()) {parameterName=PackB, samplerType=SAMPLERTYPE_LinearColor, texture=/Game/Environments/Apollo/Sets/Skyfire/Textures/T_Skyfire_PackB}
  n36 = LinearInterpolate(
   A=ScalarParameter() {defaultValue=0.75, group=Base Color, parameterName=BaseColorDirtLow}
   B=ScalarParameter() {defaultValue=1.4, group=Base Color, parameterName=BaseColorDirtHigh}
   Alpha=ChannelMaskParameter(None=@n35.RGB) {group=Base Color, maskChannel=Green, parameterName=DirtChannelMask, defaultValue=<rgba 0 1 0 0>}.R
  )
  n37 = ScalarParameter() {defaultValue=0.85, group=Base Color, parameterName=ViewAngleDependantTechPattern_Low}
  n38 = ScalarParameter() {defaultValue=1, group=Base Color, parameterName=ViewAngleDependantTechPattern_High}
  n39 = MaterialFunctionCall() {materialFunction=&MF_Fresnel_FX}
  n40 = TextureSampleParameter2D(UVs=NamedRerouteUsage()) {parameterName=PackA, samplerType=SAMPLERTYPE_LinearColor, texture=/Game/Environments/Apollo/Sets/Skyfire/Textures/T_Skyfire_PackA}
  n41 = Add(A=@n39.Fresnel, B=NamedRerouteDeclaration(None=@n40.R) {name=TechPatternBlocks, nodeColor=<rgba 0 1 0.8359 1>})
  n42 = Multiply(A=@n39.Fresnel Inverted, B=ConstantBiasScale(None=Sine(None=@n41) {period=0.5}))
  n43 = TextureSampleParameter2D(UVs=StaticSwitchParameter(True=NamedRerouteUsage(), False=TextureCoordinate()) {group=Base Color, parameterName=DiffuseColorCurveControlledByUniqueTextureUsesTechPatternUVs}) {
   group=Base Color
   parameterName=DiffuseColorCurveUniqueTexture
   samplerType=SAMPLERTYPE_LinearColor
   texture=&T_Skyfire_PackB
  }
  n44 = ChannelMaskParameter(None=@n43.RGB) {group=Base Color, parameterName=DiffuseColorCurveUniqueChannel}
  n45 = Add(
   A=Multiply(
    A=Subtract(A=@n44.R) {constB=0.5}
    B=ScalarParameter() {defaultValue=0.5, group=Base Color, parameterName=DiffuseColorCurveUniqueTextureScale}
   ) {constB=0.2}
   B=NamedRerouteUsage()
  )
  n46 = CurveAtlasRowParameter(CurveTime=StaticSwitchParameter(
    True=StaticSwitchParameter(True=@n45, False=@n44.R) {group=Base Color, parameterName=UseHeightMapAsBaseForDiffuseColorCurveControlledByUniqueTexture}
    False=NamedRerouteUsage()
   ) {group=Base Color, parameterName=DiffuseColorCurveControlledByUniqueTexture}) {
   atlas=&SkyfireCurveAtlas
   curve=/Game/Environments/Apollo/Sets/Skyfire/Curves/C_Skyfire_Diffuse_01
   group=Base Color
   parameterName=DiffuseColorCurve
  }
  n47 = Multiply(
   A=NamedRerouteDeclaration(None=LinearInterpolate(A=@n37, B=@n38, Alpha=@n42)) {name=ViewAngleDependantTechPattern, nodeColor=<rgba 1 0 0.6797 1>}
   B=@n46
  )
  NamedRerouteDeclaration(None=Multiply(A=Multiply(A=@n36, B=@n47), B=NamedRerouteUsage())) {name=BaseColor, nodeColor=<rgba 0 0.8828 1 1>}
  NamedRerouteDeclaration(None=Divide(A=NamedRerouteUsage(), B=ScalarParameter() {defaultValue=200, group=Tile, parameterName=DirtTileDivisor})) {name=DirtUVs, nodeColor=<rgba 1 0.5625 0 1>}
  n48 = ScalarParameter() {defaultValue=1, group=Ambient Occlusion, parameterName=AO_Strength}
  n49 = NamedRerouteUsage()
  NamedRerouteDeclaration(None=LinearInterpolate(B=NamedRerouteUsage(), Alpha=Max(A=@n48, B=@n49)) {constA=1}) {name=AmbientOcclusion, nodeColor=<rgba 1 0.7969 0 1>}
  NamedRerouteDeclaration(None=@n2.G) {name=MasksG_Height, nodeColor=<rgba 0 1 0 1>}
  n50 = Normalize(VectorInput=VectorParameter() {group=FakeSpecular, parameterName=FakeSpecular_LightVector, defaultValue=<rgba 0.8 -0.1 0.15 0>}.RGB)
  n51 = NamedRerouteUsage()
  n52 = ReflectionVectorWS(CustomWorldNormal=Normalize(VectorInput=Transform(None=@n51)))
  n53 = Power(Base=Abs(None=DotProduct(A=@n50, B=@n52)), Exp=ScalarParameter() {defaultValue=10, group=FakeSpecular, parameterName=FakeSpecularExponent}) {constExponent=50}
  n54 = NamedRerouteUsage()
  NamedRerouteDeclaration(None=Multiply(
    A=Multiply(A=@n53, B=@n54)
    B=VectorParameter() {group=FakeSpecular, parameterName=FakeSpecularColor, defaultValue=<rgba 0.15 0.0113 0.134 0>}.RGB
   )) {name=FakeSpecular, nodeColor=<rgba 0.4297 0 1 1>}
  n55 = MaterialFunctionCall() {materialFunction=&MF_Skyfire_RotateAboutAxis}
  NamedRerouteDeclaration(None=StaticSwitchParameter(True=@n55.Result, False=Constant()) {desc=Temp fix, group=RotateAboutAxis, parameterName=UseRotation}) {name=WPO, nodeColor=<rgba 0.6875 0 1 1>}
  NamedRerouteDeclaration(None=@n55.NormalizeRotationAxis) {name=NormalizedRotationAxis, nodeColor=<rgba 0 0.2969 1 1>}
  NamedRerouteDeclaration(None=@n55.RotationAngle) {name=RotationAngle, nodeColor=<rgba 1 0 0.7266 1>}
  NamedRerouteDeclaration(None=@n55.PivotPoint) {name=PivotPoint, nodeColor=<rgba 0.4062 1 0 1>}
  n56 = ScalarParameter() {defaultValue=220, group=Tile, parameterName=AdditionalDetailNormalTileDivisor}
  NamedRerouteDeclaration(None=Divide(A=@n6, B=@n56)) {name=AdditionalDetailNormalUVs, nodeColor=<rgba 0.1484 0 1 1>}
  NamedRerouteDeclaration(None=@n7) {name=TechTileDivisor, nodeColor=<rgba 0 1 0.7188 1>}
  n57 = ScalarParameter() {defaultValue=400, group=Tile, parameterName=MetalTileDivisorU}
  NamedRerouteDeclaration(None=@n57) {name=MetalTileDivisor.U, nodeColor=<rgba 0.4062 1 0 1>}
  NamedRerouteDeclaration(None=@n56) {name=AdditionalDetailNormalTileDivisor, nodeColor=<rgba 0.1719 1 0 1>}
  NamedRerouteDeclaration(None=@n2.B) {name=MasksB_AO, nodeColor=<rgba 0 0 1 1>}
  NamedRerouteDeclaration(None=@n2.A) {name=MasksA_CurvatureFine, nodeColor=<rgba 1 1 1 1>}
  NamedRerouteDeclaration(None=LinearInterpolate(
    A=ScalarParameter() {defaultValue=0.65, group=Metallic, parameterName=Metallic_Low}
    B=ScalarParameter() {defaultValue=0.4, group=Metallic, parameterName=Metallic_High}
    Alpha=NamedRerouteUsage()
   )) {name=Metallic, nodeColor=<rgba 1 0 0.4453 1>}
  n58 = ChannelMaskParameter(None=NamedRerouteUsage()) {group=Roughness, parameterName=RoughnessMaxMaskChannel}
  n59 = Divide(A=Time(), B=ScalarParameter() {defaultValue=18, group=Roughness, parameterName=BlocksRoughnessTimeScale})
  n60 = NamedRerouteUsage()
  n61 = ConstantBiasScale(None=Sine(None=Add(A=@n59, B=@n60)) {period=0.4})
  n62 = TextureSampleParameter2D(UVs=NamedRerouteUsage()) {parameterName=PackA, samplerType=SAMPLERTYPE_LinearColor, texture=&T_Skyfire_PackA}
  n63 = ChannelMaskParameter(None=@n62.B) {group=Roughness, maskChannel=Blue, parameterName=RoughnessChannelMask, defaultValue=<rgba 0 0 1 0>}
  n64 = TextureSampleParameter2D(UVs=NamedRerouteUsage()) {parameterName=PackA, samplerType=SAMPLERTYPE_LinearColor, texture=&T_Skyfire_PackA}
  n65 = LinearInterpolate(
   A=ScalarParameter() {defaultValue=0.28, group=Roughness, parameterName=Roughness_Low}
   B=ScalarParameter() {defaultValue=0.45, group=Roughness, parameterName=Roughness_High}
   Alpha=Multiply(A=StaticSwitchParameter(True=@n61, False=@n63.R) {group=Roughness, parameterName=UseAnimatedRoughnesMask}, B=@n64.A)
  )
  NamedRerouteDeclaration(None=StaticSwitchParameter(True=Max(A=@n58.R, B=@n65), False=@n65) {group=Roughness, parameterName=RoughnessMaxMask}) {name=Roughness, nodeColor=<rgba 0.875 1 0 1>}
  n66 = NamedRerouteUsage()
  n67 = NamedRerouteUsage()
  n68 = ScalarParameter() {defaultValue=4, group=Chroma, parameterName=ChromaFresnelMaskExponent}
  NamedRerouteDeclaration(None=StaticSwitchParameter(
    True=Constant()
    False=MaterialFunctionCall!(AmbientOcclusionMask=@n66, AmbientOcclusionExponent, ColorMask=@n67, ReflectionVector=@n52, FresnelMaskExponent=@n68) {materialFunction=&/MF_Skyfire_Chroma}.Result
   ) {desc=Temp fix, group=RotateAboutAxis, parameterName=UseRotation}) {name=Chroma, nodeColor=<rgba 0.5 0 1 1>}
  n69 = ScalarParameter() {defaultValue=50, group=Tile, parameterName=MetalTileDivisorV}
  NamedRerouteDeclaration(None=Divide(A=@n6, B=AppendVector(A=@n57, B=@n69))) {name=MetalUVs, nodeColor=<rgba 0 0.0156 1 1>}
  Transform(None=@n55.Result) {transformSourceType=TRANSFORMSOURCE_Local}
 ]}
