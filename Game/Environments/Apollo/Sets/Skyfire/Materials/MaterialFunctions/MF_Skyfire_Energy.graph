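# Material function for the Skyfire set's energy surface, outputting full material attributes with anisotropy, glowing core and view-dependent chroma.
# Core glow layers sphere masks, optional pulse, sine-animated blocks and per-instance or per-primitive offset glyphs, tinted by curve-atlas colors with abduction fades.
# Panning noise and packed textures on UV or world-aligned coordinates drive variation in color, AO and detail normals.
# Optional axis rotation drives world position offset and corrects normals to match.
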MaterialFunction /Game/Environments/Apollo/Sets/Skyfire/Materials/MaterialFunctions/MF_Skyfire_Energy
graph {nodes [
  Constant3Vector() {constant=<rgba 0 0 1 0>}
  n1 = SetMaterialAttributes!(
   MaterialAttributes
   BaseColor=NamedRerouteUsage()
   Metallic=ScalarParameter() {defaultValue=0.4, group=Metallic, parameterName=Metallic}
   Specular=ScalarParameter() {defaultValue=0.2, group=Specular, parameterName=Specular}
   Roughness=NamedRerouteUsage()
   Anisotropy=NamedRerouteUsage()
   EmissiveColor=NamedRerouteUsage()
   Tangent=NamedRerouteUsage()
   Normal=NamedRerouteUsage()
   AmbientOcclusion=NamedRerouteUsage()
   WorldPositionOffset=NamedRerouteUsage()
  ) {attributeSetTypes [
    #1
    #2
    #3
    #4
    #5
    #6
    #7
    #8
    #9
    #10
   ]}
  FunctionOutput(None=@n1)
  n2 = Panner(
   Coordinate=NamedRerouteUsage()
   Speed=AppendVector(
    A=ScalarParameter() {defaultValue=12, group=World Mask, parameterName=WorldMaskPanU}
    B=ScalarParameter() {defaultValue=9, group=World Mask, parameterName=WorldMaskPanV}
   )
  )
  n3 = ScalarParameter() {defaultValue=550, group=World Mask, parameterName=WorldMaskDivisor}
  NamedRerouteDeclaration(None=TextureSample(UVs=Divide(A=@n2, B=@n3)) {samplerSource=SSM_Wrap_WorldGroupSettings, texture=/Game/Global/Textures/Default/Noise/T_Tiling_Noise_01}.R) {name=WorldMask, nodeColor=<rgba 0 1 0.6016 1>}
  n4 = CurveAtlasRowParameter() {
   atlas=/Game/Environments/Apollo/Sets/Skyfire/Curves/SkyfireCurveAtlas
   bUseCustomPrimitiveData=true
   curve=/Game/Environments/Apollo/Sets/Skyfire/Curves/C_Skyfire_Energy_Emissive_Rare
   defaultValue=23
   parameterName=CoreGlowColor
   primitiveDataIndex=2
  }
  n5 = CurveAtlasRowParameter() {
   atlas=&SkyfireCurveAtlas
   curve=/Game/Environments/Apollo/Sets/Skyfire/Curves/C_Skyfire_Energy_Emissive_Common
   defaultValue=23
   parameterName=CoreGlowColor
   primitiveDataIndex=2
  }
  n6 = CurveAtlasRowParameter() {
   atlas=&SkyfireCurveAtlas
   curve=/Game/Environments/Apollo/Sets/Skyfire/Curves/C_Skyfire_Trim_Emissive_Uncommon
   defaultValue=26
   parameterName=AbductionGlowColor
   primitiveDataIndex=2
  }
  n7 = Multiply(A=@n6, B=ScalarParameter() {defaultValue=1, group=Core Glow, parameterName=AbductionGlowMult})
  n8 = ScalarParameter() {group=Core Glow, parameterName=AbductionFade}
  n9 = StaticSwitchParameter(
   True=@n4
   False=StaticSwitchParameter(True=LinearInterpolate(A=@n5, B=@n7, Alpha=@n8), False=@n5) {group=Core Glow, parameterName=CoreGlowColorUsesAbductionControls}
  ) {group=Core Glow, parameterName=CoreGlowColorUsesCPD}
  n10 = Add(
   A=StaticSwitchParameter(True=TextureCoordinate() {coordinateIndex=1}, False=TextureCoordinate()) {defaultValue=true, group=Core Glow, parameterName=SphereMaskUseTexCoord1}
   B=AppendVector(A=ScalarParameter() {group=Core Glow, parameterName=GlowBiasU}, B=ScalarParameter() {group=Core Glow, parameterName=GlowBiasV})
  )
  n11 = ScalarParameter() {defaultValue=0.5, group=Core Glow, parameterName=GlowCenter}
  n12 = ScalarParameter() {defaultValue=0.5, group=Core Glow, parameterName=GlowRadius}
  n13 = Multiply(
   A=MaterialFunctionCall!(
    Time
    Divisor=ScalarParameter() {defaultValue=0.2, group=Core Glow, parameterName=PulseTimeDivisor}
    Mod=ScalarParameter() {defaultValue=6, group=Core Glow, parameterName=PulseTimeMod}
   ) {materialFunction=/Game/Effects/Fort_Effects/Materials/_Functions/MF_ModuloTime}.Unclamped
   B=@n12
  )
  n14 = ScalarParameter() {group=Core Glow, parameterName=GlowHardness}
  n15 = Power(
   Base=SphereMask(
    A=SphereMask(A=@n10, B=@n11, Radius=@n13, Hardness=@n14)
    B=ScalarParameter() {defaultValue=0.5, group=Core Glow, parameterName=PulseCenter}
    Radius=ScalarParameter() {defaultValue=0.35, group=Core Glow, parameterName=PulseRadius}
    Hardness=ScalarParameter() {group=Core Glow, parameterName=PulseHardness}
   )
   Exp=ScalarParameter() {defaultValue=2.5, group=Core Glow, parameterName=PulseExponent}
  )
  n16 = ScalarParameter() {defaultValue=3, parameterName=PulseScale}
  n17 = Power(
   Base=Multiply(A=SphereMask(A=@n10, B=@n11, Radius=@n12, Hardness=@n14), B=ScalarParameter() {defaultValue=1, group=Core Glow, parameterName=CoreGlowScale})
   Exp=ScalarParameter() {defaultValue=1.5, group=Core Glow, parameterName=CoreGlowExponent}
  )
  n18 = LinearInterpolate(A=Add(A=NamedRerouteUsage()) {constB=0.5}, Alpha=@n17)
  n19 = Divide(A=Time(), B=ScalarParameter() {defaultValue=10, group=Core Glow, parameterName=BlocksEmissiveTimeScale})
  n20 = NamedRerouteUsage()
  n21 = ConstantBiasScale(None=Sine(None=Add(A=@n19, B=@n20)) {period=0.15})
  n22 = OneMinus(None=NamedRerouteUsage())
  n23 = Add(
   A=Add(A=Multiply(A=@n17, B=@n18), B=ScalarParameter() {defaultValue=0.02, group=Core Glow, parameterName=GlowBase})
   B=Multiply(A=Multiply(A=@n21, B=@n22), B=ScalarParameter() {defaultValue=0.3, group=Core Glow, parameterName=BlocksEmissiveScale})
  )
  n24 = Add(
   A=BumpOffset(Coordinate=TextureCoordinate(), Height=ScalarParameter() {defaultValue=-0.15, parameterName=GlyphBumpOffsetHeight})
   B=AppendVector(A=ScalarParameter() {group=Glyph, parameterName=GlyphBiasU}, B=ScalarParameter() {group=Glyph, parameterName=GlyphBiasV})
  )
  n25 = AppendVector(
   A=ScalarParameter() {defaultValue=1, group=Glyph, parameterName=GlyphCenterScaleU}
   B=ScalarParameter() {defaultValue=1, group=Glyph, parameterName=GlyphCenterScaleV}
  )
  n26 = Constant()
  n27 = StaticSwitchParameter(
   True=VertexInterpolator(VS=PerInstanceCustomData!(DefaultValue)).PS
   False=ScalarParameter() {bUseCustomPrimitiveData=true, parameterName=GlyphRandomValue, primitiveDataIndex=6}
  ) {group=Glyph, parameterName=GlyphRandomValueUsesPerInstanceCustomData}
  n28 = TextureSample(UVs=Add(
    A=MaterialFunctionCall(UVs=@n24, Texture Scale=@n25) {materialFunction=&ScaleUVsByCenter}.UVs
    B=AppendVector(A=@n26, B=@n27)
   )) {
   samplerSource=SSM_Wrap_WorldGroupSettings
   samplerType=SAMPLERTYPE_LinearColor
   texture=/Game/Environments/Apollo/Sets/Skyfire/Textures/T_Skyfire_Glyphs
  }
  n29 = Multiply(
   A=NamedRerouteDeclaration(None=@n28.R) {name=BigGlyph, nodeColor=<rgba 0 0.3672 1 1>}
   B=ScalarParameter() {defaultValue=2, group=Glyph, parameterName=GlyphScale}
  )
  n30 = MaterialFunctionCall(Exponent=ScalarParameter() {defaultValue=1, parameterName=GlyphFresnelExponent}) {materialFunction=&MF_Fresnel_FX}
  n31 = StaticSwitchParameter(True=Add(A=@n23, B=Multiply(A=@n29, B=@n30.Fresnel)), False=@n23) {defaultValue=true, group=Glyph, parameterName=UseGlyph}
  n32 = StaticSwitchParameter(True=Add(A=Multiply(A=@n15, B=@n16), B=@n31), False=@n31) {group=Core Glow, parameterName=UseCoreGlowPulse}
  n33 = NamedRerouteUsage()
  n34 = ScalarParameter() {defaultValue=0.4, group=Emissive, parameterName=EmissiveWorldMaskBias}
  NamedRerouteDeclaration(None=Multiply(A=Multiply(A=@n9, B=@n32), B=Add(A=@n33, B=@n34) {constB=0.4})) {name=CoreGlow, nodeColor=<rgba 0.875 1 0 1>}
  n35 = StaticSwitchParameter(
   True=TextureCoordinate()
   False=MaterialFunctionCall!(UseLocalSpace=StaticBoolParameter() {group=Tile, parameterName=UseLocalSpaceForTriPlanarTiling}, WorldPosition) {materialFunction=&/MF_WorldAlignedUVs_HardEdge}.Result
  ) {parameterName=UseStandardUVCoords}
  n36 = ScalarParameter() {defaultValue=150, group=TechPattern, parameterName=TechPatternTileDivisorU}
  n37 = ScalarParameter() {defaultValue=150, group=TechPattern, parameterName=TechPatternTileDivisorV}
  NamedRerouteDeclaration(None=Divide(A=NamedRerouteDeclaration(None=@n35) {name=CheapWorldAlignedUVs, nodeColor=<rgba 0 0.7656 1 1>}, B=AppendVector(A=@n36, B=@n37))) {name=TechPatternUVs, nodeColor=<rgba 0 1 0.7188 1>}
  n38 = TextureSampleParameter2D(UVs=NamedRerouteUsage()) {
   group=TexturePacks
   parameterName=PackA
   samplerSource=SSM_Wrap_WorldGroupSettings
   samplerType=SAMPLERTYPE_LinearColor
   texture=/Game/Environments/Apollo/Sets/Skyfire/Textures/T_Skyfire_PackA
  }
  n39 = NamedRerouteDeclaration(None=@n38.G) {name=HeightMask, nodeColor=<rgba 0 1 0.5312 1>}
  n40 = NamedRerouteUsage()
  n41 = CurveAtlasRowParameter(CurveTime=NamedRerouteDeclaration(None=Multiply(A=@n39, B=@n40)) {name=Height * WorldMask, nodeColor=<rgba 0 1 0.3906 1>}) {
   atlas=&SkyfireCurveAtlas
   curve=/Game/Environments/Apollo/Sets/Skyfire/Curves/C_Skyfire_Diffuse_Energy
   defaultValue=5
   group=Base Color
   parameterName=DiffuseColorCurve
  }
  n42 = CurveAtlasRowParameter() {
   atlas=&SkyfireCurveAtlas
   curve=/Game/Environments/Apollo/Sets/Skyfire/Curves/C_Skyfire_Emissive_Uncommon
   defaultValue=22
   parameterName=AbductionGlowColor
   primitiveDataIndex=2
  }
  NamedRerouteDeclaration(None=LinearInterpolate(A=@n41, B=@n42, Alpha=ScalarParameter() {parameterName=AbductionFade})) {name=BaseColor, nodeColor=<rgba 0.2891 0 1 1>}
  n43 = TextureSampleParameter2D(UVs=StaticSwitchParameter(True=TextureCoordinate() {coordinateIndex=1}, False=TextureCoordinate()) {group=Normals, parameterName=NormalMapUseUV1}) {
   group=Normals
   parameterName=Normal
   samplerSource=SSM_Wrap_WorldGroupSettings
   samplerType=SAMPLERTYPE_Normal
   texture=/Game/Environments/Apollo/Sets/Skyfire/Textures/T_floor_N
  }
  n44 = TextureSampleParameter2D(UVs=Divide(A=NamedRerouteUsage(), B=ScalarParameter() {defaultValue=100, group=Normals, parameterName=DetailNormalBTileDivisor})) {
   group=Normals
   parameterName=DetailNormalB
   samplerSource=SSM_Wrap_WorldGroupSettings
   samplerType=SAMPLERTYPE_Normal
   texture=/Game/Global/Textures/Default/Tileable_Patterns/T_Skyfire_DarkGlass_DetailNormal
  }
  n45 = VectorParameter() {group=Normals, parameterName=DetailNormalStrengthB, defaultValue=<rgba 0.15 0.15 1 0>}
  n46 = TextureSampleParameter2D(UVs=NamedRerouteUsage()) {
   group=Normals
   parameterName=DetailNormalA
   samplerSource=SSM_Wrap_WorldGroupSettings
   samplerType=SAMPLERTYPE_Normal
   texture=/Game/Environments/Apollo/Sets/Skyfire/Textures/T_Skyfire_BlockNormals
  }
  n47 = NamedRerouteDeclaration(None=@n46.RGB) {name=DetailNormal, nodeColor=<rgba 0 0.0391 1 1>}
  n48 = VectorParameter() {group=Normals, parameterName=DetailNormalStrengthA, defaultValue=<rgba 0.1 0.1 1 0>}
  n49 = MaterialFunctionCall(
   BaseNormal=@n43.RGB
   AdditionalNormal=LinearInterpolate(A=Multiply(A=@n44.RGB, B=@n45.RGB), B=Multiply(A=@n47, B=@n48.RGB), Alpha=NamedRerouteUsage())
  ) {materialFunction=&BlendAngleCorrectedNormals}
  n50 = Transform(None=@n49.Result)
  n51 = Add(
   A=RotateAboutAxis(NormalizedRotationAxis=NamedRerouteUsage(), RotationAngle=NamedRerouteUsage(), PivotPoint=Constant(), Position=@n50)
   B=@n50
  )
  NamedRerouteDeclaration(None=StaticSwitchParameter(True=Transform(None=@n51) {transformSourceType=TRANSFORMSOURCE_World, transformType=TRANSFORM_Tangent}, False=@n49.Result) {group=RotateAboutAxis, parameterName=UseRotation}) {name=FinalNormals, nodeColor=<rgba 0 0.7188 1 1>}
  n52 = TextureSampleParameter2D() {parameterName=Masks, samplerType=SAMPLERTYPE_LinearColor, texture=/Game/Environments/Apollo/Sets/Skyfire/Textures/T_Skyfire_EmptyMask}
  NamedRerouteDeclaration(None=@n52.RGB) {name=MasksRGB, nodeColor=<rgba 0 0.0391 1 1>}
  NamedRerouteDeclaration(None=@n52.R) {name=MasksR_Glow, nodeColor=<rgba 0 0.9062 1 1>}
  NamedRerouteDeclaration(None=@n52.G) {name=MasksG_Height, nodeColor=<rgba 0 1 0 1>}
  NamedRerouteDeclaration(None=@n52.B) {name=MasksB_AO, nodeColor=<rgba 0 0 1 1>}
  NamedRerouteDeclaration(None=@n52.A) {name=MasksA_CurvatureFine, nodeColor=<rgba 1 1 1 1>}
  n53 = NamedRerouteUsage()
  n54 = NamedRerouteUsage()
  NamedRerouteDeclaration(None=Multiply(A=Add(A=@n53, B=@n54), B=NamedRerouteUsage())) {name=Emissive, nodeColor=<rgba 0.7578 0 1 1>}
  n55 = MaterialFunctionCall!(NormalizeRotationAxis, RotationAngleOverride) {materialFunction=&/MF_Skyfire_RotateAboutAxis}
  NamedRerouteDeclaration(None=StaticSwitchParameter(True=@n55.Result, False=Constant()) {group=RotateAboutAxis, parameterName=UseRotation}) {name=WPO, nodeColor=<rgba 0.6875 0 1 1>}
  NamedRerouteDeclaration(None=@n55.NormalizeRotationAxis) {name=NormalizedRotationAxis, nodeColor=<rgba 0 0.2969 1 1>}
  NamedRerouteDeclaration(None=@n55.RotationAngle) {name=RotationAngle, nodeColor=<rgba 1 0 0.7266 1>}
  n56 = TextureSampleParameter2D(UVs=NamedRerouteUsage()) {
   group=TexturePacks
   parameterName=PackA
   samplerSource=SSM_Wrap_WorldGroupSettings
   samplerType=SAMPLERTYPE_LinearColor
   texture=&T_Skyfire_PackA
  }
  n57 = LinearInterpolate(B=@n56.B, Alpha=ScalarParameter() {defaultValue=1, group=Ambient Occlusion, parameterName=Blocks_AO}) {constA=1}
  n58 = NamedRerouteUsage()
  n59 = ScalarParameter() {defaultValue=1, group=Ambient Occlusion, parameterName=Masks_AO}
  NamedRerouteDeclaration(None=Min(A=NamedRerouteDeclaration(None=@n57) {name=BlocksAO, nodeColor=<rgba 0.8984 0 1 1>}, B=LinearInterpolate(B=@n58, Alpha=@n59) {constA=1})) {name=AO_Final, nodeColor=<rgba 1 0.5625 0 1>}
  NamedRerouteDeclaration(None=LinearInterpolate(
    A=ScalarParameter() {defaultValue=0.2, group=Roughness, parameterName=RoughnessLow}
    B=ScalarParameter() {defaultValue=0.5, group=Roughness, parameterName=RoughnessHigh}
    Alpha=NamedRerouteUsage()
   )) {name=Roughness, nodeColor=<rgba 0.9453 1 0 1>}
  n60 = NamedRerouteUsage()
  n61 = CurveAtlasRowParameter(CurveTime=DotProduct(A=Transform(None=@n60), B=CameraVectorWS())) {
   atlas=&SkyfireCurveAtlas
   curve=/Game/Environments/Apollo/Sets/Skyfire/Curves/C_Skyfire_DarkGlass_Chroma
   defaultValue=2
   group=Emissive
   parameterName=ChromaCurve
  }
  NamedRerouteDeclaration(None=Multiply(A=@n61, B=NamedRerouteUsage())) {name=Chroma, nodeColor=<rgba 0.5 0 1 1>}
  NamedRerouteDeclaration(None=Multiply(A=ScalarParameter() {defaultValue=0.8, group=Anisotropy, parameterName=Anisotropy}, B=NamedRerouteUsage())) {name=Anisotropy, nodeColor=<rgba 1 0 0.0703 1>}
 ]}
